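# Niagara effect for the PeacefulPoem Pumped cape cosmetic, adding fire and a purple glow at the cape mesh's FX_Socket_01.
# Burst-spawned looping particles follow that socket each update through skeletal mesh sampling.
# ChestFireSmall and ChestFireSoft render meshes with waveform-driven width flicker, randomized color and dynamic material parameters.
# EyeGlow combines a camera-offset sprite with light and component renderers, offsetting light positions by the sampled socket rotation.
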NiagaraSystem /BRCosmetics/Accessories/FORT_Capes/Cape_M_LRG_PeacefulPoem_PumpedUniversal/FX/NS_PeacefulPoemPumped_Cape
dependencies {
 usedDataInterfaces [&NiagaraDataInterfaceSkeletalMesh]
 usedDynamicInputs []
 usedModules [&SystemState]
}
emitters {
 ChestFireSmall {
  emitterUpdateScript [
   &EmitterState(
    Inactive Response=::NewEnumerator2
    Life Cycle Mode=::NewEnumerator1
    Loop Behavior=::NewEnumerator1
    Loop Duration Mode=::NewEnumerator1
    UseLoopDelay=false
   )
   &SpawnBurst_Instantaneous(Use Spawn Probability=false, UseLoopCountLimit=false)
  ]
  particleSpawnScript [
   &InitializeParticle(Lifetime=6, Mesh Scale Mode=::NewEnumerator3, Mesh Uniform Scale=0.5, UsePositionOffset=false, Write Material Random=true)
   &Color(Color=&MakeLinearColorFromVectorAndFloat(
     "Vector (RGB)"=&MakeVector(
      X=&RandomRangeFloat(Minimum=0, Maximum=1)
      Y=&ReturnNormalizedExecIndex()
      Z=0
     )
     "Float (Alpha)"=1
    ))
   &InitialMeshOrientation(
    Mesh Orientation Mode=::NewEnumerator2
    Orientation Vector=<0 0 1>
    Rotation=<0 0.1 0.25>
    Rotation Coordinate Space=::NewEnumerator2
    Use Rotation Vector=true
   )
   &PartitionParticles(Number Of Partitions=4) {enabled=false}
  ]
  particleUpdateScript [
   &ParticleState(
    Kill Particles When Lifetime Has Elapsed=false
    Let Infinitely Lived Particles Die When Emitter Deactivates=true
    Loop Particles Lifetime=true
   )
   &ScaleMeshSize(Scale Factor=&MakeVector(
     X=&Lerp_Float(
      A=0.8
      B=1
      Alpha=&Waveform(
       Global Input Position=&RandomRangeFloat(Minimum=0, Maximum=10)
       "Waveform [1]"=::NewEnumerator12
       "[1] Frequency"=1
       "[1] Amplitude Min/Max"=<0 1>
      )
     )
     Y=1
     Z=1
    ))
   &DynamicMaterialParameters(
    Index 0 Param 1=1
    Index 0 Param 2=&Multiply_Float(A=&RandomRangeFloat(Minimum=1, Maximum=1.2), B=1.2)
    Module.Index0Param0Write=true
    Module.Index0Param1Write=true
    Module.Index0Param2Write=true
    Module.Index0Param3Write=true
   ) {module=DynamicMaterialParameters001}
   &SkeletalMeshLocation(
    MaskSpawnGroup=false
    Mesh Sampling Type=::NewEnumerator5
    Skeletal Mesh=DataInterface() {propertyValues {
      DefaultMesh=/BRCosmetics/Accessories/FORT_Capes/Cape_M_LRG_PeacefulPoem_PumpedUniversal/Meshes/M_LRG_PeacefulPoem_PumpedUniversal_Cape
      ExcludeBoneName=root
      OptionalComponentClass=null
      PreviewMesh=&M_LRG_PeacefulPoem_PumpedUniversal_Cape
      SkinningMode=SkinOnTheFly
      SoftSourceActor=null
      SourceMode=Default
      UvSetIndex=0
      WholeMeshLOD=-1
      bExcludeBone=false
      bReadDeformedGeometry=true
      bRequireCurrentFrameData=true
      ComponentTags []
      FilteredBones []
      FilteredSockets [FX_Socket_01]
      MeshUserParameter {parameter {name=None, typeDefHandle {registeredTypeIndex=1}, varData []}}
      SamplingRegions []
     }}
    Socket Sampling Mode=::NewEnumerator5
   )
  ]
  rendererClasses [&NiagaraMeshRendererProperties]
  renderers [{rendererClass=&NiagaraMeshRendererProperties, rendererIndex=0}]
 }
 ChestFireSoft {
  emitterUpdateScript [
   &EmitterState(
    Inactive Response=::NewEnumerator2
    Life Cycle Mode=::NewEnumerator1
    Loop Behavior=::NewEnumerator1
    Loop Duration Mode=::NewEnumerator1
    UseLoopDelay=false
   )
   &SpawnBurst_Instantaneous(Use Spawn Probability=false, UseLoopCountLimit=false)
  ]
  particleSpawnScript [
   &InitializeParticle(Lifetime=6, Mesh Scale Mode=::NewEnumerator3, Mesh Uniform Scale=0.5, UsePositionOffset=false, Write Material Random=true)
   &Color(Color=&MakeLinearColorFromVectorAndFloat(
     "Vector (RGB)"=&MakeVector(
      X=&RandomRangeFloat(Minimum=0, Maximum=1)
      Y=&ReturnNormalizedExecIndex()
      Z=0
     )
     "Float (Alpha)"=1
    ))
   &InitialMeshOrientation(
    Mesh Orientation Mode=::NewEnumerator2
    Orientation Vector=<0 0 1>
    Rotation=<0 0.1 0.25>
    Rotation Coordinate Space=::NewEnumerator2
    Use Rotation Vector=true
   )
   &PartitionParticles(Number Of Partitions=4) {enabled=false}
  ]
  particleUpdateScript [
   &ParticleState(
    Kill Particles When Lifetime Has Elapsed=false
    Let Infinitely Lived Particles Die When Emitter Deactivates=true
    Loop Particles Lifetime=true
   )
   &ScaleMeshSize(Scale Factor=&MakeVector(
     X=&Lerp_Float(
      A=0.8
      B=1
      Alpha=&Waveform(
       Global Input Position=&RandomRangeFloat(Minimum=0, Maximum=10)
       "Waveform [1]"=::NewEnumerator12
       "[1] Frequency"=1
       "[1] Amplitude Min/Max"=<0 1>
      )
     )
     Y=1
     Z=1
    ))
   &DynamicMaterialParameters(
    Index 0 Param 1=1
    Index 0 Param 2=&Multiply_Float(A=&RandomRangeFloat(Minimum=1, Maximum=1.2), B=1.2)
    Module.Index0Param0Write=true
    Module.Index0Param1Write=true
    Module.Index0Param2Write=true
    Module.Index0Param3Write=true
   ) {module=DynamicMaterialParameters001}
   &SkeletalMeshLocation(
    MaskSpawnGroup=false
    Mesh Sampling Type=::NewEnumerator5
    Sampled Position Offset=<0 0 0>
    Skeletal Mesh=DataInterface() {propertyValues {
      DefaultMesh=&M_LRG_PeacefulPoem_PumpedUniversal_Cape
      ExcludeBoneName=root
      OptionalComponentClass=null
      PreviewMesh=&M_LRG_PeacefulPoem_PumpedUniversal_Cape
      SkinningMode=SkinOnTheFly
      SoftSourceActor=null
      SourceMode=Default
      UvSetIndex=0
      WholeMeshLOD=-1
      bExcludeBone=false
      bReadDeformedGeometry=true
      bRequireCurrentFrameData=true
      ComponentTags []
      FilteredBones []
      FilteredSockets [FX_Socket_01]
      MeshUserParameter {parameter {name=None, typeDefHandle {registeredTypeIndex=1}, varData []}}
      SamplingRegions []
     }}
    Socket Sampling Mode=::NewEnumerator5
   )
  ]
  rendererClasses [&NiagaraMeshRendererProperties]
  renderers [{rendererClass=&NiagaraMeshRendererProperties, rendererIndex=0}]
 }
 EyeGlow {
  emitterUpdateScript [
   &EmitterState(Life Cycle Mode=::NewEnumerator1, Loop Behavior=::NewEnumerator1, Loop Duration Mode=::NewEnumerator1, UseLoopDelay=false)
   &SpawnBurst_Instantaneous(Use Spawn Probability=false, UseLoopCountLimit=true)
  ]
  particleSpawnScript [&InitializeParticle(Color=<rgba 0.5 0.1 0.8 1>, Sprite Size Mode=::NewEnumerator3, Uniform Sprite Size=12, UsePositionOffset=false, Write Material Random=true)]
  particleUpdateScript [
   &ParticleState(
    Kill Particles When Lifetime Has Elapsed=false
    Let Infinitely Lived Particles Die When Emitter Deactivates=true
    Loop Particles Lifetime=true
   )
   &SkeletalMeshLocation(
    MaskSpawnGroup=false
    Mesh Sampling Type=::NewEnumerator5
    Skeletal Mesh=DataInterface() {propertyValues {
      DefaultMesh=&M_LRG_PeacefulPoem_PumpedUniversal_Cape
      ExcludeBoneName=root
      OptionalComponentClass=null
      PreviewMesh=&M_LRG_PeacefulPoem_PumpedUniversal_Cape
      SkinningMode=SkinOnTheFly
      SoftSourceActor=null
      SourceMode=Default
      UvSetIndex=0
      WholeMeshLOD=-1
      bExcludeBone=false
      bReadDeformedGeometry=true
      bRequireCurrentFrameData=true
      ComponentTags []
      FilteredBones []
      FilteredSockets [FX_Socket_01]
      MeshUserParameter {parameter {name=None, typeDefHandle {registeredTypeIndex=1}, varData []}}
      SamplingRegions []
     }}
    Socket Sampling Mode=::NewEnumerator5
   )
   &ScaleColor(Scale Alpha=0.1, Scale RGB=&VectorFromFloat(Value=3.5), ScaleA=true, ScaleRGB=true)
   &DynamicMaterialParameters(Module.Index0Param0Write=true, Module.Index0Param1Write=true, Module.Index0Param2Write=true, Module.Index0Param3Write=true)
   &FN_Light_Attributes(
    Color=<rgba 0.5 0.1 0.8 1>
    Create Color Attribute=true
    Create Exponent Attribute=true
    Create Position Attribute=true
    Create Radius Attribute=true
    Enable Particle Light Enable=true
    Exponent=1
    Light Enabled=true
    Position=&AddVectorToPosition(
     Position=$Particles.Position
     Vector=&RotateVectorByQuaternion(Vector To Rotate=<0 -10 0>, Quaternion=$Particles.SkeletalMeshLocation.SampledRotation)
    )
    Radius=30
    Scale Color=25
   )
   &FN_Gen5_Light(
    Attenutation=10
    Color:LinearColor=<rgba 0.5 0.1 0.8 1>
    Intensity=20
    Position Offset:Vector3f=&RotateVectorByQuaternion(Vector To Rotate=<0 -3 0>, Quaternion=$Particles.SkeletalMeshLocation.SampledRotation)
   )
   &CameraOffset(Camera Offset Amount=2)
  ]
  rendererClasses [
   &NiagaraSpriteRendererProperties
   &NiagaraLightRendererProperties
   &NiagaraComponentRendererProperties
  ]
  renderers [
   {rendererClass=&NiagaraSpriteRendererProperties}
   {rendererClass=&NiagaraLightRendererProperties, rendererIndex=1}
   {rendererClass=&NiagaraComponentRendererProperties, rendererIndex=2}
  ]
 }
}
summary {
 systemName=NS_PeacefulPoemPumped_Cape
 emitters [
  {emitterName=ChestFireSoft, rendererClasses [&NiagaraMeshRendererProperties]}
  {emitterName=ChestFireSmall, rendererClasses [&NiagaraMeshRendererProperties]}
  {
   emitterName=EyeGlow
   rendererClasses [
    &NiagaraSpriteRendererProperties
    &NiagaraLightRendererProperties
    &NiagaraComponentRendererProperties
   ]
  }
 ]
 userVariables [User.RibbonIDOffset {type=NiagaraInt32}]
}
