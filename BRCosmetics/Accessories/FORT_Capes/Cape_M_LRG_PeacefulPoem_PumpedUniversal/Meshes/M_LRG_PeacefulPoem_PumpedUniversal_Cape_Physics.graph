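# Physics asset for the PeacefulPoem PumpedUniversal cape accessory, giving its dyn_cape bones capsule bodies for physical motion.
# Center, left and right chains descend from tiny MSTR root capsules, with mid branches off l_03 and r_04 and an edge branch off r_01.
# Chain joints carry tight swing and twist limits.
# Two angularly free constraints tie l_03 and r_03 to center bone 03, linking the three chains laterally.
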PhysicsAsset /BRCosmetics/Accessories/FORT_Capes/Cape_M_LRG_PeacefulPoem_PumpedUniversal/Meshes/M_LRG_PeacefulPoem_PumpedUniversal_Cape_Physics
bodies {
 dyn_cape_01 [{shapeType=Capsule, length=12.7556, radius=4.9773, shapeName=dyn_cape_01_capsule, center=<7.15 0 1.5269>, rotation=<rot 90 0 0>}]
 dyn_cape_02 [{shapeType=Capsule, length=12.6645, radius=5.0683, shapeName=dyn_cape_01_capsule, center=<7.15 0 1.5269>, rotation=<rot 90 0 0>}]
 dyn_cape_03 [{shapeType=Capsule, length=13.4256, radius=6.0692, shapeName=dyn_cape_01_capsule, center=<7.15 0 0.0944>, rotation=<rot 90 0 0>}]
 dyn_cape_MSTR [{shapeType=Capsule, length=0.505, radius=0.505, shapeName=dyn_cape_r_MSTR_capsule, rotation=<rot 0 -90 0>}]
 dyn_cape_l_01 [{shapeType=Capsule, length=13.4743, radius=4.2585, shapeName=dyn_cape_01_capsule, center=<7.15 0 1.5269>, rotation=<rot 90 0 0>}]
 dyn_cape_l_02 [{shapeType=Capsule, length=13.5083, radius=4.2246, shapeName=dyn_cape_01_capsule, center=<7.15 0 1.5269>, rotation=<rot 90 0 0>}]
 dyn_cape_l_03 [{shapeType=Capsule, length=12.4145, radius=5.3183, shapeName=dyn_cape_01_capsule, center=<7.15 0 1.5269>, rotation=<rot 90 0 0>}]
 dyn_cape_l_04 [{shapeType=Capsule, length=8.4458, radius=4.1308, shapeName=dyn_cape_01_capsule, center=<5.4856 0 2.9006>, rotation=<rot 90 0 0>}]
 dyn_cape_l_05 [{shapeType=Capsule, length=8.0083, radius=3.9121, shapeName=dyn_cape_01_capsule, center=<7.15 0 2.9006>, rotation=<rot 90 0 0>}]
 dyn_cape_l_MSTR [{shapeType=Capsule, length=0.505, radius=0.505, shapeName=dyn_cape_r_MSTR_capsule, rotation=<rot 0 -90 0>}]
 dyn_cape_l_mid_01 [{shapeType=Capsule, length=9.9692, radius=5.5921, shapeName=dyn_cape_l_mid_01_capsule, center=<8.7745 0 1.5269>, rotation=<rot 90 0 0>}]
 dyn_cape_l_mid_02 [{shapeType=Capsule, length=5.5477, radius=4.9038, shapeName=dyn_cape_l_mid_02_capsule, center=<10.3451 0 2.9006>, rotation=<rot 90 0 0>}]
 dyn_cape_l_mid_03 [{shapeType=Capsule, length=4.3779, radius=4.2362, shapeName=dyn_cape_l_mid_03_capsule, center=<6.7472 0 2.9006>, rotation=<rot 90 0 0>}]
 dyn_cape_l_mid_04 [{shapeType=Capsule, length=14.4354, radius=4.8682, shapeName=dyn_cape_l_mid_04_capsule, center=<11.2275 0 2.9006>, rotation=<rot 90 0 0>}]
 dyn_cape_l_mid_05 [{
   shapeType=Capsule
   length=15.6044
   radius=6.4446
   shapeName=dyn_cape_l_mid_05_capsule
   center=<11.2548 -1.4613 0.3147>
   rotation=<rot 90 180 176.5753>
  }]
 dyn_cape_l_mid_06 [{shapeType=Capsule, length=8.5169, radius=6.7402, shapeName=dyn_cape_l_mid_06_capsule, center=<12.6387 0 2.4746>, rotation=<rot 90 0 0>}]
 dyn_cape_l_mid_07 [{shapeType=Capsule, length=7.801, radius=6.1443, shapeName=dyn_cape_l_mid_07_capsule, center=<7.8952 0 2.9006>, rotation=<rot 90 0 0>}]
 dyn_cape_l_mid_08 [{shapeType=Capsule, length=7.6831, radius=4.5201, shapeName=dyn_cape_l_mid_08_capsule, center=<7.3276 0 2.9006>, rotation=<rot 90 0 0>}]
 dyn_cape_l_mid_09 [{shapeType=Capsule, length=8.0754, radius=4.2888, shapeName=dyn_cape_l_mid_09_capsule, center=<4.9449 0 2.9006>, rotation=<rot 90 0 0>}]
 dyn_cape_r_01 [{shapeType=Capsule, length=11.7556, radius=5.9773, shapeName=dyn_cape_01_capsule, center=<7.15 0 1.5269>, rotation=<rot 90 0 0>}]
 dyn_cape_r_02 [{shapeType=Capsule, length=12.7895, radius=4.9433, shapeName=dyn_cape_01_capsule, center=<7.15 0 1.5269>, rotation=<rot 90 0 0>}]
 dyn_cape_r_03 [{shapeType=Capsule, length=10.3706, radius=7.3623, shapeName=dyn_cape_01_capsule, center=<7.15 0 1.5269>, rotation=<rot 90 0 0>}]
 dyn_cape_r_04 [{shapeType=Capsule, length=9.2897, radius=8.4431, shapeName=dyn_cape_01_capsule, center=<7.2328 0 3.8398>, rotation=<rot 90 0 0>}]
 dyn_cape_r_05 [{shapeType=Capsule, length=17.727, radius=3.6308, shapeName=dyn_cape_01_capsule, center=<7.15 0 3.8576>, rotation=<rot 90 180 -176.3539>}]
 dyn_cape_r_06 [{shapeType=Capsule, length=14.727, radius=5.0683, shapeName=dyn_cape_01_capsule, center=<10.1396 0 3.8576>, rotation=<rot 90 0 0>}]
 dyn_cape_r_07 [{shapeType=Capsule, length=16.9431, radius=3.7871, shapeName=dyn_cape_01_capsule, center=<11.5861 0 3.8576>, rotation=<rot 90 180 -171.9566>}]
 dyn_cape_r_MSTR [{shapeType=Capsule, length=0.505, radius=0.505, shapeName=dyn_cape_r_MSTR_capsule, rotation=<rot 0 -90 0>}]
 dyn_cape_r_edge_01 [{shapeType=Capsule, length=7.2748, radius=3.0174, shapeName=dyn_cape_r_edge_01_capsule, center=<9.3267 0 1.5269>, rotation=<rot 90 0 0>}]
 dyn_cape_r_edge_02 [{shapeType=Capsule, length=9.502, radius=2.4054, shapeName=dyn_cape_r_edge_02_capsule, center=<5.8546 0 1.5269>, rotation=<rot 90 0 0>}]
 dyn_cape_r_mid_01 [{
   shapeType=Capsule
   length=5.8967
   radius=4.8788
   shapeName=dyn_cape_r_mid_01_capsule
   center=<5.8196 0 2.9006>
   rotation=<rot 90 180 178.8926>
  }]
 dyn_cape_r_mid_02 [{shapeType=Capsule, length=9.2683, radius=4.8745, shapeName=dyn_cape_r_mid_02_capsule, center=<3.2334 0 2.9006>, rotation=<rot 90 0 0>}]
 dyn_cape_r_mid_03 [{shapeType=Capsule, length=9.8422, radius=4.6334, shapeName=dyn_cape_r_mid_03_capsule, center=<6.1282 0 2.9006>, rotation=<rot 90 0 0>}]
 dyn_cape_r_mid_04 [{shapeType=Capsule, length=7.6275, radius=4.178, shapeName=dyn_cape_r_mid_04_capsule, center=<10.5734 0 3.8398>, rotation=<rot 90 0 0>}]
 dyn_cape_r_mid_05 [{shapeType=Capsule, length=5.0247, radius=6.2165, shapeName=dyn_cape_r_mid_05_capsule, center=<9.5339 0 3.2438>, rotation=<rot 90 0 0>}]
 dyn_cape_r_mid_06 [{shapeType=Capsule, length=8.6444, radius=4.4004, shapeName=dyn_cape_r_mid_06_capsule, center=<8.7329 0 3.2438>, rotation=<rot 90 0 0>}]
}
constraints [
 {bone1Name=dyn_cape_r_01, bone2Name=dyn_cape_r_MSTR, swing1LimitDegrees=5, swing2LimitDegrees=20, twistLimitDegrees=5}
 {bone1Name=dyn_cape_r_02, bone2Name=dyn_cape_r_01, swing1LimitDegrees=5, swing2LimitDegrees=20, twistLimitDegrees=5}
 {bone1Name=dyn_cape_r_03, bone2Name=dyn_cape_r_02, swing1LimitDegrees=5, swing2LimitDegrees=20, twistLimitDegrees=5}
 {bone1Name=dyn_cape_r_04, bone2Name=dyn_cape_r_03, swing1LimitDegrees=5, swing2LimitDegrees=20, twistLimitDegrees=5}
 {bone1Name=dyn_cape_r_05, bone2Name=dyn_cape_r_04, swing1LimitDegrees=5, swing2LimitDegrees=20, twistLimitDegrees=5}
 {bone1Name=dyn_cape_r_06, bone2Name=dyn_cape_r_05, swing1LimitDegrees=5, swing2LimitDegrees=20, twistLimitDegrees=5}
 {bone1Name=dyn_cape_r_07, bone2Name=dyn_cape_r_06, swing1LimitDegrees=5, swing2LimitDegrees=20, twistLimitDegrees=5}
 {bone1Name=dyn_cape_l_01, bone2Name=dyn_cape_l_MSTR, swing1LimitDegrees=5, swing2LimitDegrees=20, twistLimitDegrees=5}
 {bone1Name=dyn_cape_l_02, bone2Name=dyn_cape_l_01, swing1LimitDegrees=5, swing2LimitDegrees=20, twistLimitDegrees=5}
 {bone1Name=dyn_cape_l_03, bone2Name=dyn_cape_l_02, swing1LimitDegrees=5, swing2LimitDegrees=20, twistLimitDegrees=5}
 {bone1Name=dyn_cape_l_04, bone2Name=dyn_cape_l_03, swing1LimitDegrees=4, swing2LimitDegrees=20, twistLimitDegrees=5}
 {bone1Name=dyn_cape_l_05, bone2Name=dyn_cape_l_04, swing1LimitDegrees=5, swing2LimitDegrees=20, twistLimitDegrees=5}
 {bone1Name=dyn_cape_01, bone2Name=dyn_cape_MSTR, swing1LimitDegrees=5, swing2LimitDegrees=20, twistLimitDegrees=5}
 {bone1Name=dyn_cape_02, bone2Name=dyn_cape_01, swing1LimitDegrees=5, swing2LimitDegrees=20, twistLimitDegrees=5}
 {bone1Name=dyn_cape_03, bone2Name=dyn_cape_02, swing1LimitDegrees=5, swing2LimitDegrees=20, twistLimitDegrees=5}
 {bone1Name=dyn_cape_l_mid_01, bone2Name=dyn_cape_l_03, swing1LimitDegrees=5, swing2LimitDegrees=20, twistLimitDegrees=5}
 {bone1Name=dyn_cape_l_mid_02, bone2Name=dyn_cape_l_mid_01, swing1LimitDegrees=5, swing2LimitDegrees=20, twistLimitDegrees=5}
 {bone1Name=dyn_cape_l_mid_03, bone2Name=dyn_cape_l_mid_02, swing1LimitDegrees=5, swing2LimitDegrees=20, twistLimitDegrees=5}
 {bone1Name=dyn_cape_l_mid_04, bone2Name=dyn_cape_l_mid_03, swing1LimitDegrees=5, swing2LimitDegrees=20, twistLimitDegrees=5}
 {bone1Name=dyn_cape_l_mid_05, bone2Name=dyn_cape_l_03, swing1LimitDegrees=2, swing2LimitDegrees=20, twistLimitDegrees=5}
 {bone1Name=dyn_cape_l_mid_06, bone2Name=dyn_cape_l_mid_05, swing1LimitDegrees=5, swing2LimitDegrees=20, twistLimitDegrees=5}
 {bone1Name=dyn_cape_l_mid_07, bone2Name=dyn_cape_l_mid_06, swing1LimitDegrees=5, swing2LimitDegrees=20, twistLimitDegrees=5}
 {bone1Name=dyn_cape_l_mid_08, bone2Name=dyn_cape_l_mid_07, swing1LimitDegrees=5, swing2LimitDegrees=20, twistLimitDegrees=5}
 {bone1Name=dyn_cape_l_mid_09, bone2Name=dyn_cape_l_mid_08, swing1LimitDegrees=5, swing2LimitDegrees=20, twistLimitDegrees=5}
 {bone1Name=dyn_cape_r_mid_01, bone2Name=dyn_cape_r_04, swing1LimitDegrees=5, swing2LimitDegrees=20, twistLimitDegrees=5}
 {bone1Name=dyn_cape_r_mid_02, bone2Name=dyn_cape_r_mid_01, swing1LimitDegrees=5, swing2LimitDegrees=20, twistLimitDegrees=5}
 {bone1Name=dyn_cape_r_mid_03, bone2Name=dyn_cape_r_mid_02, swing1LimitDegrees=5, swing2LimitDegrees=20, twistLimitDegrees=5}
 {bone1Name=dyn_cape_r_mid_04, bone2Name=dyn_cape_r_04, swing1LimitDegrees=5, swing2LimitDegrees=20, twistLimitDegrees=5}
 {bone1Name=dyn_cape_r_mid_05, bone2Name=dyn_cape_r_mid_04, swing1LimitDegrees=5, swing2LimitDegrees=20, twistLimitDegrees=5}
 {bone1Name=dyn_cape_r_mid_06, bone2Name=dyn_cape_r_mid_05, swing1LimitDegrees=5, swing2LimitDegrees=20, twistLimitDegrees=5}
 {bone1Name=dyn_cape_r_edge_01, bone2Name=dyn_cape_r_01, swing1LimitDegrees=5, swing2LimitDegrees=20, twistLimitDegrees=5}
 {bone1Name=dyn_cape_r_edge_02, bone2Name=dyn_cape_r_edge_01, swing1LimitDegrees=5, swing2LimitDegrees=20, twistLimitDegrees=5}
 {
  bone1Name=dyn_cape_r_03
  bone2Name=dyn_cape_03
  swing1LimitDegrees=45
  swing1Motion=Free
  swing2LimitDegrees=45
  swing2Motion=Free
  twistLimitDegrees=45
  twistMotion=Free
 }
 {
  bone1Name=dyn_cape_l_03
  bone2Name=dyn_cape_03
  swing1LimitDegrees=45
  swing1Motion=Free
  swing2LimitDegrees=45
  swing2Motion=Free
  twistLimitDegrees=45
  twistMotion=Free
 }
]
